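# Modular Control Rig biped spine module providing FK and spline IK controls for a connector-defined bone chain.
# Construction spawns FK, Hip and Chest controls, IK and Stretch channels, and tangents solved so a B-spline passes through the first four bones after the pelvis.
# A spline solve places bones along the four IK controls using stored offsets, interpolated up vectors and length compensation when stretch is off.
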ControlRigBlueprint /EpicControlRig/Modules/Biped/Spine_MH/CRM_Epic_SpineMH_v01
graphs [
 ConstructionGraph {nodes [
   n1 = /Script/RigVM.RigVMFunction_MathTransformArrayToSRT(Transforms=/Script/ControlRig.RigUnit_GetTransformItemArray(Space=GlobalSpace, bInitial=False, Items=RigVMVariableNode() {variable=Bones}.Value) {is=struct RigVMUnitNode}.Transforms) {is=struct RigVMUnitNode}
   RigVMFunction_Sequence = RigVMAggregateNode(ExecuteContext=@n11.ExecuteContext)
   BSpline Passing thru 4 Points_1_1 = RigVMFunctionReferenceNode(
    Auto Weights=False
    Weight 1=0.333
    Weight 2=0.666
    Point 1=RigVMDispatchNode(Index=1, Array=@n1.Translations) {template="DISPATCH_RigVMDispatch_ArrayGetAtIndex(in Array,in Index,out Element)"}.Element
    Point 2=RigVMDispatchNode(Index=2, Array=@n1.Translations) {template="DISPATCH_RigVMDispatch_ArrayGetAtIndex(in Array,in Index,out Element)"}.Element
    Point 3=RigVMDispatchNode(Index=3, Array=@n1.Translations) {template="DISPATCH_RigVMDispatch_ArrayGetAtIndex(in Array,in Index,out Element)"}.Element
    Point 0=RigVMDispatchNode(Index=0, Array=@n1.Translations) {template="DISPATCH_RigVMDispatch_ArrayGetAtIndex(in Array,in Index,out Element)"}.Element
    Execute=@RigVMFunction_Sequence.C
   )
   RigVMVariableNode(
    ExecuteContext=@"BSpline Passing thru 4 Points_1_1".Execute
    Value=/Script/ControlRigSpline.RigUnit_GetLengthControlRigSpline(Spline=@"BSpline Passing thru 4 Points_1_1".Spline) {is=struct RigVMUnitNode}.Length
   ) {variable=Length}
   > n2 = /Script/ControlRig.RigUnit_HierarchyAddControlTransform(
    InitialValue=(Rotation=(X=0,Y=0,Z=0,W=1),Translation=(X=0,Y=0,Z=0),Scale3D=(X=1,Y=1,Z=1))
    Name=Hips Tan
    OffsetSpace=GlobalSpace
    OffsetTransform=(Rotation=(X=0,Y=0,Z=0,W=1),Translation=(X=0,Y=0,Z=0),Scale3D=(X=1,Y=1,Z=1))
    Settings="(InitialSpace=LocalSpace,bUsePreferredRotationOrder=False,PreferredRotationOrder=YZX,Limits=(LimitTranslationX=(bMinimum=False,bMaximum=False),LimitTranslationY=(bMinimum=False,bMaximum=False),LimitTranslationZ=(bMinimum=False,bMaximum=False),LimitPitch=(b ...(truncated)"
    OffsetTransform.Translation=RigVMDispatchNode(Index=1, Array=@"BSpline Passing thru 4 Points_1_1".Control Points) {template="DISPATCH_RigVMDispatch_ArrayGetAtIndex(in Array,in Index,out Element)"}.Element
    Parent=RigVMVariableNode() {variable=Hip Control}.Value
    Settings.Shape.Color=RigVMVariableNode() {variable=Control Color}.Value
    Settings.Shape.Transform=RigVMVariableNode() {variable=IK Control Shape Transform}.Value
    Settings.Shape.Name=RigVMVariableNode() {variable=IK Control Shape}.Value
   ) {
    is=struct RigVMUnitNode
    template="SpawnControl::Execute(in InitialValue,in Settings,in OffsetTransform,in OffsetSpace,in Parent,in Name,out Item)"
   }
   > n3 = &RigUnit_HierarchyAddControlTransform(
    InitialValue=(Rotation=(X=0,Y=0,Z=0,W=1),Translation=(X=0,Y=0,Z=0),Scale3D=(X=1,Y=1,Z=1))
    Name=Chest Tan
    OffsetSpace=GlobalSpace
    OffsetTransform=(Rotation=(X=0,Y=0,Z=0,W=1),Translation=(X=0,Y=0,Z=0),Scale3D=(X=1,Y=1,Z=1))
    Settings="(InitialSpace=LocalSpace,bUsePreferredRotationOrder=False,PreferredRotationOrder=YZX,Limits=(LimitTranslationX=(bMinimum=False,bMaximum=False),LimitTranslationY=(bMinimum=False,bMaximum=False),LimitTranslationZ=(bMinimum=False,bMaximum=False),LimitPitch=(b ...(truncated)"
    OffsetTransform.Translation=RigVMDispatchNode(Index=2, Array=@"BSpline Passing thru 4 Points_1_1".Control Points) {template="DISPATCH_RigVMDispatch_ArrayGetAtIndex(in Array,in Index,out Element)"}.Element
    Parent=RigVMVariableNode() {variable=Chest Control}.Value
    Settings.Shape.Color=RigVMVariableNode() {variable=Control Color}.Value
    Settings.Shape.Transform=RigVMVariableNode() {variable=IK Control Shape Transform}.Value
    Settings.Shape.Name=RigVMVariableNode() {variable=IK Control Shape}.Value
   ) {template="SpawnControl::Execute(in InitialValue,in Settings,in OffsetTransform,in OffsetSpace,in Parent,in Name,out Item)"}
   > RigVMVariableNode(
    Value=((Type=None,Name="None"),(Type=None,Name="None"),(Type=None,Name="None"),(Type=None,Name="None"))
    Value.0=RigVMVariableNode() {variable=Hip Control}.Value
    Value.3=RigVMVariableNode() {variable=Chest Control}.Value
    Value.1=@n2.Item
    Value.2=@n3.Item
   ) {variable=IK Controls}
   > /Script/ControlRig.RigUnit_SetChannelHosts(Channel=RigVMVariableNode() {variable=IK Channel}.Value, Hosts=RigVMVariableNode() {variable=FK Controls}.Value) {is=struct RigVMUnitNode}
   n4 = RigVMDispatchNode(Array=RigVMVariableNode() {variable=IK Controls}.Value) {template="DISPATCH_RigVMDispatch_ArrayClone(in Array,out Clone)"}
   > n5 = RigVMDispatchNode(
    Array=@n4.Clone
    Index=RigVMDispatchNode(Array=@n4.Clone, Element=RigVMVariableNode() {variable=Hip Control}.Value) {template="DISPATCH_RigVMDispatch_ArrayFind(in Array,in Element,out Index,out Success)"}.Index
   ) {template="DISPATCH_RigVMDispatch_ArrayRemove(io Array,in Index)"}
   > &RigUnit_SetChannelHosts(Hosts=@n5.Array, Channel=RigVMVariableNode() {variable=IK Channel}.Value)
   > &RigUnit_SetChannelHosts(Hosts=@n5.Array, Channel=RigVMVariableNode() {variable=Stretch Channel}.Value)
   RigVMDispatchNode(ExecuteContext=(), Array=RigVMVariableNode() {variable=FK Controls}.Value) {template="DISPATCH_RigVMDispatch_ArrayReset(io Array)"}
   > RigVMVariableNode(Value=RigVMVariableNode() {variable=Parent}.Value) {variable=Running Parent}
   > For_Each = RigVMDispatchNode(Array=RigVMVariableNode() {variable=Bones}.Value) {template="DISPATCH_RigVMDispatch_ArrayIterator(in Array,out Element,out Index,out Count,out Ratio)"}
   RigVMFunction_ControlFlowBranch = &RigVMFunction_ControlFlowBranch(
    Condition=RigVMDispatchNode(B=0, A=@For_Each.Index) {template="DISPATCH_RigVMDispatch_CoreEquals(in A,in B,out Result)"}.Result
    ExecuteContext=@For_Each.ExecuteContext
   )
   Subtract = /Script/RigVM.RigVMFunction_MathIntSub(B=1, A=@For_Each.Index) {is=struct RigVMUnitNode, template="Subtract::Execute(in A,in B,out Result)"}
   Int to Name = /Script/RigVM.RigVMFunction_MathIntToName(PaddedSize=0, Number=@Subtract.Result) {is=struct RigVMUnitNode, template="Int to Name::Execute(in Number,in PaddedSize,out Result)"}
   Concat = /Script/RigVM.RigVMFunction_NameConcat(A="Spine ", B=@"Int to Name".Result) {is=struct RigVMUnitNode, template="Concat::Execute(in A,in B,out Result)"}
   n6 = &RigUnit_GetTransform(bInitial=False, Item=@For_Each.Element)
   Minimum = /Script/RigVM.RigVMFunction_MathIntMin(B=2, A=@For_Each.Index) {is=struct RigVMUnitNode, template="Minimum::Execute(in A,in B,out Result)"}
   n7 = RigVMDispatchNode(Index=@Minimum.Result, Array=RigVMVariableNode() {variable=Lengths}.Value) {template="DISPATCH_RigVMDispatch_ArrayGetAtIndex(in Array,in Index,out Element)"}
   Multiply = &RigVMFunction_MathDoubleMul(B=0.09, A=@n7.Element)
   n8 = &RigUnit_HierarchyAddControlTransform(
    InitialValue=(Rotation=(X=0,Y=0,Z=0,W=1),Translation=(X=0,Y=0,Z=0),Scale3D=(X=1,Y=1,Z=1))
    OffsetSpace=GlobalSpace
    OffsetTransform=(Rotation=(X=0,Y=0,Z=0,W=1),Translation=(X=0,Y=0,Z=0),Scale3D=(X=1,Y=1,Z=1))
    Settings="(InitialSpace=LocalSpace,bUsePreferredRotationOrder=False,PreferredRotationOrder=YZX,Limits=(LimitTranslationX=(bMinimum=False,bMaximum=False),LimitTranslationY=(bMinimum=False,bMaximum=False),LimitTranslationZ=(bMinimum=False,bMaximum=False),LimitPitch=(b ...(truncated)"
    Name=@Concat.Result
    Settings.Shape.Transform.Rotation=@n6.Transform.Rotation
    OffsetTransform.Translation=@n6.Transform.Translation
    Settings.Shape.Transform.Scale3D.X=@Multiply.Result
    Settings.Shape.Transform.Translation.Z=&RigVMFunction_MathDoubleMul(B=0.5, A=@n7.Element).Result
    ExecutePin=@RigVMFunction_ControlFlowBranch.False
    Parent=RigVMVariableNode() {variable=Running Parent}.Value
   ) {template="SpawnControl::Execute(in InitialValue,in Settings,in OffsetTransform,in OffsetSpace,in Parent,in Name,out Item)"}
   > RigVMVariableNode(Value=@n8.Item) {variable=Running Parent}
   > Add = RigVMDispatchNode(Element=@n8.Item, Array=RigVMVariableNode() {variable=FK Controls}.Value) {template="DISPATCH_RigVMDispatch_ArrayAdd(io Array,in Element,out Index)"}
   RigUnit_PrepareForExecution = &RigUnit_PrepareForExecution()
   n9 = /Script/ControlRig.RigUnit_HierarchyGetChainItemArray(End=(Type=Connector,Name="End"), Start=(Type=Connector,Name="Root"), bIncludeEnd=True, bIncludeStart=True, bReverse=False) {is=struct RigVMUnitNode}
   RigVMVariableNode(
    Value=RigVMDispatchNode(Index=0, Array=@n9.Chain) {template="DISPATCH_RigVMDispatch_ArrayGetAtIndex(in Array,in Index,out Element)"}.Element
    ExecuteContext=@RigUnit_PrepareForExecution.ExecutePin
   ) {variable=Pelvis Bone}
   > n10 = RigVMDispatchNode(Index=0, Array=@n9.Chain) {template="DISPATCH_RigVMDispatch_ArrayRemove(io Array,in Index)"}
   > RigVMVariableNode(Value=@n10.Array) {variable=Bones}
   > Get Distances Between = RigVMCollapseNode(Initial=False, Items=RigVMVariableNode() {variable=Bones}.Value)
   > RigVMVariableNode(Value=@"Get Distances Between".Distances) {variable=Lengths}
   > n11 = RigVMVariableNode(Value=(Type=Connector,Name="Parent")) {variable=Parent}
   At = RigVMDispatchNode(Index=0, Array=RigVMVariableNode() {variable=Bones}.Value) {template="DISPATCH_RigVMDispatch_ArrayGetAtIndex(in Array,in Index,out Element)"}
   SpawnControl = &RigUnit_HierarchyAddControlTransform(
    ExecutePin=()
    InitialValue=(Rotation=(X=0,Y=0,Z=0,W=1),Translation=(X=0,Y=0,Z=0),Scale3D=(X=1,Y=1,Z=1))
    Name=Hips
    OffsetSpace=GlobalSpace
    OffsetTransform=(Rotation=(X=0,Y=0,Z=0,W=1),Translation=(X=0,Y=0,Z=0),Scale3D=(X=1,Y=1,Z=1))
    Settings="(InitialSpace=LocalSpace,bUsePreferredRotationOrder=False,PreferredRotationOrder=YZX,Limits=(LimitTranslationX=(bMinimum=False,bMaximum=False),LimitTranslationY=(bMinimum=False,bMaximum=False),LimitTranslationZ=(bMinimum=False,bMaximum=False),LimitPitch=(b ...(truncated)"
    OffsetTransform.Translation=&RigUnit_GetTransform(bInitial=False, Item=@At.Element).Transform.Translation
    Parent=RigVMVariableNode() {variable=Parent}.Value
   ) {template="SpawnControl::Execute(in InitialValue,in Settings,in OffsetTransform,in OffsetSpace,in Parent,in Name,out Item)"}
   > RigVMVariableNode(Value=@SpawnControl.Item) {variable=Hip Control}
   n12 = RigVMVariableNode() {variable=Bones}
   > n13 = &RigUnit_HierarchyAddControlTransform(
    InitialValue=(Rotation=(X=0,Y=0,Z=0,W=1),Translation=(X=0,Y=0,Z=0),Scale3D=(X=1,Y=1,Z=1))
    Name=Chest
    OffsetSpace=GlobalSpace
    OffsetTransform=(Rotation=(X=0,Y=0,Z=0,W=1),Translation=(X=0,Y=0,Z=0),Scale3D=(X=1,Y=1,Z=1))
    Settings="(InitialSpace=LocalSpace,bUsePreferredRotationOrder=False,PreferredRotationOrder=YZX,Limits=(LimitTranslationX=(bMinimum=False,bMaximum=False),LimitTranslationY=(bMinimum=False,bMaximum=False),LimitTranslationZ=(bMinimum=False,bMaximum=False),LimitPitch=(b ...(truncated)"
    OffsetTransform.Translation=&RigUnit_GetTransform(
     bInitial=False
     Item=RigVMDispatchNode(Index=-1, Array=@n12.Value) {template="DISPATCH_RigVMDispatch_ArrayGetAtIndex(in Array,in Index,out Element)"}.Element
    ).Transform.Translation
    Parent=RigVMVariableNode() {variable=Parent}.Value
   ) {template="SpawnControl::Execute(in InitialValue,in Settings,in OffsetTransform,in OffsetSpace,in Parent,in Name,out Item)"}
   > n14 = RigVMVariableNode(Value=@n13.Item) {variable=Chest Control}
   RigVMFunction_MathTransformArrayToSRT = &RigVMFunction_MathTransformArrayToSRT(Transforms=&RigUnit_GetTransformItemArray(Space=GlobalSpace, bInitial=False, Items=RigVMVariableNode() {variable=Bones}.Value).Transforms)
   BSpline Passing thru 4 Points_1 = RigVMFunctionReferenceNode(
    Auto Weights=False
    Weight 1=0.333
    Weight 2=0.666
    Point 0=RigVMDispatchNode(Index=0, Array=@RigVMFunction_MathTransformArrayToSRT.Translations) {template="DISPATCH_RigVMDispatch_ArrayGetAtIndex(in Array,in Index,out Element)"}.Element
    Point 1=RigVMDispatchNode(Index=1, Array=@RigVMFunction_MathTransformArrayToSRT.Translations) {template="DISPATCH_RigVMDispatch_ArrayGetAtIndex(in Array,in Index,out Element)"}.Element
    Point 2=RigVMDispatchNode(Index=2, Array=@RigVMFunction_MathTransformArrayToSRT.Translations) {template="DISPATCH_RigVMDispatch_ArrayGetAtIndex(in Array,in Index,out Element)"}.Element
    Point 3=RigVMDispatchNode(Index=3, Array=@RigVMFunction_MathTransformArrayToSRT.Translations) {template="DISPATCH_RigVMDispatch_ArrayGetAtIndex(in Array,in Index,out Element)"}.Element
    Execute=@n14.ExecuteContext
   )
   RigVMVariableNode(
    ExecuteContext=@"BSpline Passing thru 4 Points_1".Execute
    Value=&RigUnit_GetLengthControlRigSpline(Spline=@"BSpline Passing thru 4 Points_1".Spline).Length
   ) {variable=Length}
   n15 = RigVMVariableNode() {variable=Hip Control}
   > n16 = &RigUnit_HierarchyAddControlTransform(
    InitialValue=(Rotation=(X=0,Y=0,Z=0,W=1),Translation=(X=0,Y=0,Z=0),Scale3D=(X=1,Y=1,Z=1))
    Name=Hips Tan
    OffsetSpace=GlobalSpace
    OffsetTransform=(Rotation=(X=0,Y=0,Z=0,W=1),Translation=(X=0,Y=0,Z=0),Scale3D=(X=1,Y=1,Z=1))
    Settings="(InitialSpace=LocalSpace,bUsePreferredRotationOrder=False,PreferredRotationOrder=YZX,Limits=(LimitTranslationX=(bMinimum=False,bMaximum=False),LimitTranslationY=(bMinimum=False,bMaximum=False),LimitTranslationZ=(bMinimum=False,bMaximum=False),LimitPitch=(b ...(truncated)"
    OffsetTransform.Translation=RigVMDispatchNode(Index=1, Array=@"BSpline Passing thru 4 Points_1".Control Points) {template="DISPATCH_RigVMDispatch_ArrayGetAtIndex(in Array,in Index,out Element)"}.Element
    Parent=@n15.Value
   ) {template="SpawnControl::Execute(in InitialValue,in Settings,in OffsetTransform,in OffsetSpace,in Parent,in Name,out Item)"}
   n17 = RigVMVariableNode() {variable=Chest Control}
   > n18 = &RigUnit_HierarchyAddControlTransform(
    InitialValue=(Rotation=(X=0,Y=0,Z=0,W=1),Translation=(X=0,Y=0,Z=0),Scale3D=(X=1,Y=1,Z=1))
    Name=Chest Tan
    OffsetSpace=GlobalSpace
    OffsetTransform=(Rotation=(X=0,Y=0,Z=0,W=1),Translation=(X=0,Y=0,Z=0),Scale3D=(X=1,Y=1,Z=1))
    Settings="(InitialSpace=LocalSpace,bUsePreferredRotationOrder=False,PreferredRotationOrder=YZX,Limits=(LimitTranslationX=(bMinimum=False,bMaximum=False),LimitTranslationY=(bMinimum=False,bMaximum=False),LimitTranslationZ=(bMinimum=False,bMaximum=False),LimitPitch=(b ...(truncated)"
    OffsetTransform.Translation=RigVMDispatchNode(Index=2, Array=@"BSpline Passing thru 4 Points_1".Control Points) {template="DISPATCH_RigVMDispatch_ArrayGetAtIndex(in Array,in Index,out Element)"}.Element
    Parent=@n17.Value
   ) {template="SpawnControl::Execute(in InitialValue,in Settings,in OffsetTransform,in OffsetSpace,in Parent,in Name,out Item)"}
   > RigVMVariableNode(
    Value=((Type=None,Name="None"),(Type=None,Name="None"),(Type=None,Name="None"),(Type=None,Name="None"))
    Value.1=@n16.Item
    Value.2=@n18.Item
    Value.0=@n15.Value
    Value.3=@n17.Value
   ) {variable=IK Controls}
   Clone = RigVMDispatchNode(Array=RigVMVariableNode() {variable=Bones}.Value) {template="DISPATCH_RigVMDispatch_ArrayClone(in Array,out Clone)"}
   n19 = RigVMDispatchNode(Index=0, Array=@Clone.Clone, ExecuteContext=@RigVMFunction_Sequence.A) {template="DISPATCH_RigVMDispatch_ArrayRemove(io Array,in Index)"}
   For_Each_1 = RigVMDispatchNode(Array=@n19.Array, ExecuteContext=@n32.ExecuteContext) {template="DISPATCH_RigVMDispatch_ArrayIterator(in Array,out Element,out Index,out Count,out Ratio)"}
   n20 = RigVMDispatchNode(Index=@For_Each_1.Index, Array=RigVMVariableNode() {variable=FK Controls}.Value) {template="DISPATCH_RigVMDispatch_ArrayGetAtIndex(in Array,in Index,out Element)"}
   > RigVMDispatchNode(
    Element=&RigVMFunction_MathTransformMakeRelative(
     Global=&RigUnit_GetTransform(bInitial=False, Item=@For_Each_1.Element).Transform
     Parent=&RigUnit_GetTransform(bInitial=False, Item=@n20.Element).Transform
    ).Local
    Array=RigVMVariableNode() {variable=Bone Deltas}.Value
   ) {template="DISPATCH_RigVMDispatch_ArrayAdd(io Array,in Element,out Index)"}
   n21 = RigVMDispatchNode(ExecuteContext=@n26.ExecuteContext, Array=RigVMVariableNode() {variable=Bones}.Value) {template="DISPATCH_RigVMDispatch_ArrayIterator(in Array,out Element,out Index,out Count,out Ratio)"}
   Spline From Element Setup = RigVMCollapseNode(Items=RigVMVariableNode() {variable=IK Controls}.Value)
   > RigVMDispatchNode(
    Element=&RigVMFunction_MathTransformMakeRelative(
     Global=&RigUnit_GetTransform(Item=@n21.Element).Transform
     Parent=/Script/ControlRigSpline.RigUnit_TransformFromControlRigSpline(UpVector=(X=0,Y=1,Z=0), U=@n21.Ratio, Spline=@"Spline From Element Setup".Spline) {is=struct RigVMUnitNode}.Transform
    ).Local
    Array=RigVMVariableNode() {variable=IK Offsets}.Value
   ) {template="DISPATCH_RigVMDispatch_ArrayAdd(io Array,in Element,out Index)"}
   RigVMRerouteNode(Value=())
   n22 = RigVMVariableNode() {variable=Bones}
   Get Chain Length = RigVMFunctionReferenceNode(Initial=true, ExecuteContext=@n23.ExecuteContext, Items=@n22.Value)
   Num = RigVMDispatchNode(Array=@n22.Value) {template="DISPATCH_RigVMDispatch_ArrayGetNum(in Array,out Num)"}
   n23 = RigVMDispatchNode(
    Array=RigVMDispatchNode(Array=RigVMVariableNode() {variable=FK Controls}.Value) {template="DISPATCH_RigVMDispatch_ArrayClone(in Array,out Clone)"}.Clone
    Other=RigVMVariableNode() {variable=IK Controls}.Value
    ExecuteContext=@RigVMFunction_Sequence.E
   ) {template="DISPATCH_RigVMDispatch_ArrayUnion(io Array,in Other)"}
   Set Control Scale = RigVMFunctionReferenceNode(
    ExecuteContext=@"Get Chain Length".ExecuteContext
    Auto Scale Factor=RigVMDispatchNode(
     False=1
     True=&RigVMFunction_MathDoubleDiv(B=22.5, A=@"Get Chain Length".Length).Result
     Condition=/Script/RigVM.RigVMFunction_MathIntGreater(B=1, A=@Num.Num) {is=struct RigVMUnitNode, template="Greater::Execute(in A,in B,out Result)"}.Result
    ) {template="DISPATCH_RigVMDispatch_If(in Condition,in True,in False,out Result)"}.Result
    Global Scale=RigVMVariableNode() {variable=Global Control Scale}.Value
    Scale=RigVMVariableNode() {variable=Control Scale}.Value
    Controls=@n23.Array
   )
   n24 = RigVMVariableNode() {variable=Bones}
   n25 = &RigUnit_HierarchyAddNull(
    Name=Chest Space
    Space=GlobalSpace
    Parent=RigVMDispatchNode(Index=-1, Array=RigVMVariableNode() {variable=FK Controls}.Value) {template="DISPATCH_RigVMDispatch_ArrayGetAtIndex(in Array,in Index,out Element)"}.Element
    Transform=&RigUnit_GetTransform(Item=RigVMDispatchNode(Index=-1, Array=@n24.Value) {template="DISPATCH_RigVMDispatch_ArrayGetAtIndex(in Array,in Index,out Element)"}.Element).Transform
    ExecutePin=@n31.ExecuteContext
   )
   > RigVMVariableNode(Value=@n25.Item) {variable=Chest Space Null}
   RigVMVariableNode() {variable=IK Control Shape}
   n26 = RigVMDispatchNode(Array=RigVMVariableNode() {variable=IK Offsets}.Value, ExecuteContext=@RigVMFunction_Sequence.D) {template="DISPATCH_RigVMDispatch_ArrayReset(io Array)"}
   n27 = RigVMDispatchNode(Index=0, Array=RigVMVariableNode() {variable=Bones}.Value) {template="DISPATCH_RigVMDispatch_ArrayGetAtIndex(in Array,in Index,out Element)"}
   n28 = RigVMDispatchNode(Index=-1, Array=RigVMVariableNode() {variable=Bones}.Value) {template="DISPATCH_RigVMDispatch_ArrayGetAtIndex(in Array,in Index,out Element)"}
   Control Stack at Item_1 = RigVMFunctionReferenceNode(
    Add Bottom Null=False
    Add Null=false
    Add Secondary Control=False
    Name=Hip
    Orient to World=true
    Secondary Settings="(Shape=(bVisible=True,Name=\"Default\",Color=(R=1,G=0,B=0,A=1),Transform=(Rotation=(X=0,Y=0,Z=0,W=1),Translation=(X=0,Y=0,Z=0),Scale3D=(X=1,Y=1,Z=1))),Proxy=(bIsProxy=False,ShapeVisibility=BasedOnSelection),DisplayName=\"None\",InitialSpace=LocalSpace,bUsePref ...(truncated)"
    Settings="(Shape=(bVisible=True,Name=\"Box_Solid\",Color=(R=1,G=0,B=0,A=1),Transform=(Rotation=(X=0,Y=0,Z=0,W=1),Translation=(X=0,Y=0,Z=0),Scale3D=(X=1.3,Y=1,Z=0.1))),Proxy=(bIsProxy=False,ShapeVisibility=BasedOnSelection),DisplayName=\"None\",InitialSpace=LocalSpace,bU ...(truncated)"
    Item=@n27.Element
    Settings.Shape.Color=RigVMVariableNode() {variable=Control Color}.Value
    Parent=RigVMVariableNode() {variable=Parent}.Value
    ExecuteContext=@RigVMFunction_Sequence.B
   )
   > RigVMVariableNode(Value=@"Control Stack at Item_1".Control) {variable=Hip Control}
   > SpawnAnimationChannel = /Script/ControlRig.RigUnit_HierarchyAddAnimationChannelBool(InitialValue=False, LimitsEnabled=(), MaximumValue=True, MinimumValue=False, Name=IK, Parent=@"Control Stack at Item_1".Control) {
    is=struct RigVMUnitNode
    template="SpawnAnimationChannel::Execute(in InitialValue,in MinimumValue,in MaximumValue,in LimitsEnabled,in Parent,in Name,out Item)"
   }
   > RigVMVariableNode(Value=@SpawnAnimationChannel.Item) {variable=IK Channel}
   > n29 = &RigUnit_HierarchyAddAnimationChannelBool(InitialValue=False, LimitsEnabled=(), MaximumValue=True, MinimumValue=False, Name=Stretch, Parent=@"Control Stack at Item_1".Control) {template="SpawnAnimationChannel::Execute(in InitialValue,in MinimumValue,in MaximumValue,in LimitsEnabled,in Parent,in Name,out Item)"}
   > RigVMVariableNode(Value=@n29.Item) {variable=Stretch Channel}
   n30 = RigVMVariableNode() {variable=IK Control Shape Transform}
   Scale = /Script/RigVM.RigVMFunction_MathVectorScale(Factor=2, Value=@n30.Value.Scale3D) {is=struct RigVMUnitNode, template="Scale::Execute(in Value,in Factor,out Result)"}
   > Control Stack at Item_1_1 = RigVMFunctionReferenceNode(
    Add Bottom Null=False
    Add Null=false
    Add Secondary Control=False
    Name=Chest
    Null Suffix=Nul
    Orient to World=true
    Secondary Settings="(Shape=(bVisible=True,Name=\"Default\",Color=(R=1,G=0,B=0,A=1),Transform=(Rotation=(X=0,Y=0,Z=0,W=1),Translation=(X=0,Y=0,Z=0),Scale3D=(X=1,Y=1,Z=1))),Proxy=(bIsProxy=False,ShapeVisibility=BasedOnSelection),DisplayName=\"None\",InitialSpace=LocalSpace,bUsePref ...(truncated)"
    Settings="(Shape=(bVisible=True,Name=\"Default\",Color=(R=1,G=0,B=0,A=1),Transform=(Rotation=(X=0,Y=0,Z=0,W=1),Translation=(X=0,Y=0,Z=0),Scale3D=(X=1,Y=1,Z=1))),Proxy=(bIsProxy=False,ShapeVisibility=BasedOnSelection),DisplayName=\"None\",InitialSpace=LocalSpace,bUsePref ...(truncated)"
    Item=@n28.Element
    Settings.Shape.Transform.Scale3D=@Scale.Result
    Settings.Shape.Color=RigVMVariableNode() {variable=Control Color}.Value
    Settings.Shape.Transform.Rotation=@n30.Value.Rotation
    Settings.Shape.Transform.Translation=@n30.Value.Translation
    Settings.Shape.Name=RigVMVariableNode() {variable=IK Control Shape}.Value
    Parent=RigVMVariableNode() {variable=Parent}.Value
   )
   > n31 = RigVMVariableNode(Value=@"Control Stack at Item_1_1".Control) {variable=Chest Control}
   Create FK Chain Controls _1 = RigVMFunctionReferenceNode(
    Control Settings="(DisplayName=\"None\",InitialSpace=LocalSpace,bUsePreferredRotationOrder=False,PreferredRotationOrder=YZX,Limits=(LimitTranslationX=(bMinimum=False,bMaximum=False),LimitTranslationY=(bMinimum=False,bMaximum=False),LimitTranslationZ=(bMinimum=False,bMaximum=F ...(truncated)"
    Name="Spine "
    Orient to World=true
    Items=@n19.Array
    ExecuteContext=@n19.ExecuteContext
    Parent=RigVMVariableNode() {variable=Parent}.Value
    Control Settings.Shape.Color=RigVMVariableNode() {variable=Control Color}.Value
    Control Settings.Shape.Name=RigVMVariableNode() {variable=FK Control Shape}.Value
    Control Settings.Shape.Transform=RigVMVariableNode() {variable=FK Control Shape Transform}.Value
   )
   > RigVMVariableNode(Value=@"Create FK Chain Controls _1".Controls) {variable=FK Controls}
   > n32 = RigVMDispatchNode(Array=RigVMVariableNode() {variable=Bone Deltas}.Value) {template="DISPATCH_RigVMDispatch_ArrayReset(io Array)"}
  ]}
 ContainedGraph {nodes [
   n1 = RigVMDispatchNode(ExecuteContext=(), Array=RigVMVariableNode() {variable=Bones}.Value) {template="DISPATCH_RigVMDispatch_ArrayIterator(in Array,out Element,out Index,out Count,out Ratio)"}
   MathIntEquals = RigVMDispatchNode(A=@n1.Index, B=&RigVMFunction_MathIntSub(B=1, A=@n1.Count) {template="Subtract::Execute(in A,in B,out Result)"}.Result) {template="DISPATCH_RigVMDispatch_CoreEquals(in A,in B,out Result)"}
   n2 = &RigVMFunction_ControlFlowBranch(ExecuteContext=@n1.ExecuteContext, Condition=@MathIntEquals.Result)
   n3 = /Script/ControlRig.RigUnit_AimBoneMath(
    DebugSettings=(Scale=10,WorldOffset=(Rotation=(X=0,Y=0,Z=0,W=1),Translation=(X=0,Y=0,Z=0),Scale3D=(X=1,Y=1,Z=1)))
    Primary=(Weight=1,Axis=(X=1,Y=0,Z=0),Target=(X=0,Y=0,Z=0),Kind=Location,Space=(Type=Bone))
    Secondary=(Weight=1,Axis=(X=0,Y=0,Z=1),Target=(X=0,Y=0,Z=1),Kind=Direction,Space=(Type=Bone))
    Weight=1
    Primary.Space=RigVMDispatchNode(
     Index=/Script/RigVM.RigVMFunction_MathIntAdd(B=1, A=@n1.Index) {is=struct RigVMUnitNode, template="Add::Execute(in A,in B,out Result)"}.Result
     Array=RigVMVariableNode() {variable=Bones}.Value
    ) {template="DISPATCH_RigVMDispatch_ArrayGetAtIndex(in Array,in Index,out Element)"}.Element
    Secondary.Space=@n1.Element
    InputTransform=&RigUnit_GetTransform(Item=@n1.Element).Transform
   ) {is=struct RigVMUnitNode}
   &RigUnit_SetTransform(Weight=1, bPropagateToChildren=false, Value=@n3.Result, Item=@n1.Element, ExecutePin=@n2.False)
   n4 = RigVMFunctionEntryNode()
   SequenceExecution = RigVMAggregateNode(ExecuteContext=@n4.Completed)
   n5 = RigVMDispatchNode(ExecuteContext=@SequenceExecution.A, Array=RigVMVariableNode() {variable=Bones}.Value) {template="DISPATCH_RigVMDispatch_ArrayIterator(in Array,out Element,out Index,out Count,out Ratio)"}
   Is First or Last = RigVMCollapseNode(Count=@n5.Count, Index=@n5.Index)
   n6 = RigVMVariableNode() {variable=IK Controls}
   Spline From Element = RigVMCollapseNode(Items=@n6.Value)
   n7 = &RigUnit_GetLengthControlRigSpline(Spline=@"Spline From Element".Spline)
   n8 = RigVMVariableNode() {variable=Default Length}
   n9 = RigVMDispatchNode(
    True=1
    False=/Script/RigVM.RigVMFunction_MathFloatMin(
     B=1
     A=/Script/RigVM.RigVMFunction_MathFloatDiv(B=@n7.Length, A=@n8.Value) {is=struct RigVMUnitNode, template="Divide::Execute(in A,in B,out Result)"}.Result
    ) {is=struct RigVMUnitNode, template="Minimum::Execute(in A,in B,out Result)"}.Result
    Condition=RigVMVariableNode() {variable=Enable Stretch}.Value
   ) {template="DISPATCH_RigVMDispatch_If(in Condition,in True,in False,out Result)"}
   UpVector = RigVMCollapseNode(
    Start=RigVMDispatchNode(Index=0, Array=@n6.Value) {template="DISPATCH_RigVMDispatch_ArrayGetAtIndex(in Array,in Index,out Element)"}.Element
    End=RigVMDispatchNode(Index=3, Array=@n6.Value) {template="DISPATCH_RigVMDispatch_ArrayGetAtIndex(in Array,in Index,out Element)"}.Element
    T=@n5.Ratio
   )
   n10 = /Script/RigVM.RigVMFunction_MathTransformMakeAbsolute(
    Local=RigVMDispatchNode(Index=@n5.Index, Array=RigVMVariableNode() {variable=Spine Offsets}.Value) {template="DISPATCH_RigVMDispatch_ArrayGetAtIndex(in Array,in Index,out Element)"}.Element
    Parent=&RigUnit_TransformFromControlRigSpline(
     Roll=0
     U=&RigVMFunction_MathFloatMul(B=@n5.Ratio, A=@n9.Result) {template="Multiply::Execute(in A,in B,out Result)"}.Result
     Spline=@"Spline From Element".Spline
     UpVector=@UpVector.Up Vector
    ).Transform
   ) {is=struct RigVMUnitNode, template="Make Absolute::Execute(in Local,in Parent,out Global)"}
   n11 = RigVMDispatchNode(Index=@n5.Index, Array=RigVMVariableNode() {variable=IK Controls}.Value) {template="DISPATCH_RigVMDispatch_ArrayGetAtIndex(in Array,in Index,out Element)"}
   > &RigUnit_SetTransform(
    Weight=1
    Item=@n5.Element
    Value=RigVMDispatchNode(
     True=(Rotation=(X=0,Y=0,Z=0,W=1),Translation=(X=0,Y=0,Z=0),Scale3D=(X=0,Y=0,Z=0))
     Condition=@"Is First or Last".Result
     False=@n10.Global
     True.Scale3D=@n10.Global.Scale3D
     True.Translation=@n10.Global.Translation
     True.Rotation=/Script/ControlRig.RigUnit_ProjectTransformToNewParent(bChildInitial=True, bOldParentInitial=True, NewParent=@n11.Element, OldParent=@n11.Element, Child=@n5.Element) {is=struct RigVMUnitNode}.Transform.Rotation
    ) {template="DISPATCH_RigVMDispatch_If(in Condition,in True,in False,out Result)"}.Result
   )
   RigVMFunctionReturnNode(Completed=@SequenceExecution.D)
  ]}
 ContainedGraph {nodes [
   n1 = &RigUnit_GetTransformItemArray(Space=GlobalSpace, Items=RigVMFunctionEntryNode().Items)
   RigVMFunctionReturnNode(Spline=/Script/ControlRigSpline.RigUnit_ControlRigSplineFromPoints(
     Points=((X=0,Y=0,Z=0),(X=0,Y=0,Z=0),(X=0,Y=0,Z=0),(X=0,Y=0,Z=0))
     SamplesPerSegment=16
     SplineMode=BSpline
     Points.0=RigVMDispatchNode(Index=0, Array=@n1.Transforms) {template="DISPATCH_RigVMDispatch_ArrayGetAtIndex(in Array,in Index,out Element)"}.Element.Translation
     Points.1=RigVMDispatchNode(Index=1, Array=@n1.Transforms) {template="DISPATCH_RigVMDispatch_ArrayGetAtIndex(in Array,in Index,out Element)"}.Element.Translation
     Points.2=RigVMDispatchNode(Index=2, Array=@n1.Transforms) {template="DISPATCH_RigVMDispatch_ArrayGetAtIndex(in Array,in Index,out Element)"}.Element.Translation
     Points.3=RigVMDispatchNode(Index=3, Array=@n1.Transforms) {template="DISPATCH_RigVMDispatch_ArrayGetAtIndex(in Array,in Index,out Element)"}.Element.Translation
    ) {is=struct RigVMUnitNode}.Spline)
  ]}
 ContainedGraph {nodes [
   n1 = RigVMFunctionEntryNode()
   n2 = &RigUnit_GetTransform(Item=@n1.Start)
   n3 = &RigUnit_GetTransform(Item=@n1.End)
   RigVMFunctionReturnNode(Up Vector=/Script/RigVM.RigVMFunction_MathQuaternionRotateVector(
     Vector=(X=0,Y=1,Z=0)
     Transform=/Script/RigVM.RigVMFunction_MathQuaternionSlerp(T=@n1.T, A=@n2.Transform.Rotation, B=@n3.Transform.Rotation) {is=struct RigVMUnitNode, template="Interpolate::Execute(in A,in B,in T,out Result)"}.Result
    ) {is=struct RigVMUnitNode, template="Rotate Vector::Execute(in Transform,in Vector,out Result)"}.Result)
  ]}
 ContainedGraph {nodes [
   n1 = RigVMFunctionEntryNode()
   MathIntEquals = RigVMDispatchNode(B=0, A=@n1.Index) {template="DISPATCH_RigVMDispatch_CoreEquals(in A,in B,out Result)"}
   MathIntEquals_1 = RigVMDispatchNode(A=@n1.Index, B=&RigVMFunction_MathIntSub(B=1, A=@n1.Count) {template="Subtract::Execute(in A,in B,out Result)"}.Result) {template="DISPATCH_RigVMDispatch_CoreEquals(in A,in B,out Result)"}
   RigVMFunctionReturnNode(Result=&RigVMFunction_MathBoolOr(A=@MathIntEquals.Result, B=@MathIntEquals_1.Result).Result)
  ]}
 ContainedGraph {nodes [
   n1 = RigVMFunctionEntryNode()
   SequenceExecution = &RigVMFunction_Sequence(ExecuteContext=@n1.ExecuteContext)
   SequenceExecution_1 = &RigVMFunction_Sequence(ExecuteContext=@SequenceExecution.B)
   SequenceExecution_2 = &RigVMFunction_Sequence(ExecuteContext=@SequenceExecution_1.B)
   RigVMFunctionReturnNode(A=@SequenceExecution.A, B=@SequenceExecution_1.A, C=@SequenceExecution_2.A, D=@SequenceExecution_2.B)
  ]}
 ContainedGraph {nodes [
   n1 = &RigUnit_GetTransformItemArray(Space=GlobalSpace, Items=RigVMFunctionEntryNode().Items)
   RigVMFunctionReturnNode(Spline=&RigUnit_ControlRigSplineFromPoints(
     Points=((X=0,Y=0,Z=0),(X=0,Y=0,Z=0),(X=0,Y=0,Z=0),(X=0,Y=0,Z=0))
     SamplesPerSegment=16
     SplineMode=BSpline
     Points.0=RigVMDispatchNode(Index=0, Array=@n1.Transforms) {template="DISPATCH_RigVMDispatch_ArrayGetAtIndex(in Array,in Index,out Element)"}.Element.Translation
     Points.1=RigVMDispatchNode(Index=1, Array=@n1.Transforms) {template="DISPATCH_RigVMDispatch_ArrayGetAtIndex(in Array,in Index,out Element)"}.Element.Translation
     Points.2=RigVMDispatchNode(Index=2, Array=@n1.Transforms) {template="DISPATCH_RigVMDispatch_ArrayGetAtIndex(in Array,in Index,out Element)"}.Element.Translation
     Points.3=RigVMDispatchNode(Index=3, Array=@n1.Transforms) {template="DISPATCH_RigVMDispatch_ArrayGetAtIndex(in Array,in Index,out Element)"}.Element.Translation
    ).Spline)
  ]}
 ContainedGraph {nodes [
   n1 = RigVMFunctionEntryNode()
   SequenceAggregate = &RigVMFunction_Sequence(ExecuteContext=@n1.ExecuteContext)
   DISPATCH_RigVMDispatch_ArrayReset = RigVMDispatchNode(ExecuteContext=@SequenceAggregate.A, Array=RigVMVariableNode() {variable=Values}.Value) {template="DISPATCH_RigVMDispatch_ArrayReset(io Array)"}
   n2 = RigVMVariableNode() {variable=Items}
   > DISPATCH_RigVMDispatch_ArrayIterator = RigVMDispatchNode(Array=@n2.Value) {template="DISPATCH_RigVMDispatch_ArrayIterator(in Array,out Element,out Index,out Count,out Ratio)"}
   DISPATCH_RigVMDispatch_CoreEquals = RigVMDispatchNode(B=0, A=@DISPATCH_RigVMDispatch_ArrayIterator.Index) {template="DISPATCH_RigVMDispatch_CoreEquals(in A,in B,out Result)"}
   RigVMFunction_ControlFlowBranch = &RigVMFunction_ControlFlowBranch(ExecuteContext=@DISPATCH_RigVMDispatch_ArrayIterator.ExecuteContext, Condition=@DISPATCH_RigVMDispatch_CoreEquals.Result)
   n3 = RigVMVariableNode() {variable=Initial}
   Subtract = &RigVMFunction_MathIntSub(B=1, A=@DISPATCH_RigVMDispatch_ArrayIterator.Index) {template="Subtract::Execute(in A,in B,out Result)"}
   DISPATCH_RigVMDispatch_ArrayGetAtIndex = RigVMDispatchNode(Index=@Subtract.Result, Array=@n2.Value) {template="DISPATCH_RigVMDispatch_ArrayGetAtIndex(in Array,in Index,out Element)"}
   RigVMFunction_MathVectorDistance = &RigVMFunction_MathVectorDistance(
    A=&RigUnit_GetTransform(Item=@DISPATCH_RigVMDispatch_ArrayIterator.Element, bInitial=@n3.Value).Transform.Translation
    B=&RigUnit_GetTransform(Item=@DISPATCH_RigVMDispatch_ArrayGetAtIndex.Element, bInitial=@n3.Value).Transform.Translation
   )
   DISPATCH_RigVMDispatch_ArrayAdd = RigVMDispatchNode(
    Array=@DISPATCH_RigVMDispatch_ArrayReset.Array
    ExecuteContext=@RigVMFunction_ControlFlowBranch.False
    Element=@RigVMFunction_MathVectorDistance.Result
   ) {template="DISPATCH_RigVMDispatch_ArrayAdd(io Array,in Element,out Index)"}
   RigVMFunctionReturnNode(Distances=@DISPATCH_RigVMDispatch_ArrayReset.Array, ExecuteContext=@SequenceAggregate.B)
   RigVMVariableNode() {variable=Values}
   RigVMVariableNode() {variable=Values}
  ]}
 ContainedGraph {nodes [
   n1 = RigVMFunctionEntryNode()
   n2 = RigVMDispatchNode(Num=4, Array=RigVMVariableNode() {variable=Outputs}.Value, ExecuteContext=@n1.Execute) {template="DISPATCH_RigVMDispatch_ArraySetNum(io Array,in Num)"}
   n3 = RigVMVariableNode() {variable=Point 0}
   > n4 = RigVMDispatchNode(Index=0, Array=@n2.Array, Element=@n3.Value) {template="DISPATCH_RigVMDispatch_ArraySetAtIndex(io Array,in Index,in Element)"}
   n5 = RigVMVariableNode() {variable=Point 3}
   n6 = &RigVMFunction_MathVectorDistance(B=@n1.Point 0, A=@n1.Point 1)
   n7 = &RigVMFunction_MathFloatAdd(A=@n6.Result, B=&RigVMFunction_MathVectorDistance(A=@n1.Point 1, B=@n1.Point 2).Result) {template="Add::Execute(in A,in B,out Result)"}
   n8 = &RigVMFunction_MathFloatAdd(A=@n7.Result, B=&RigVMFunction_MathVectorDistance(A=@n1.Point 2, B=@n1.Point 3).Result) {template="Add::Execute(in A,in B,out Result)"}
   n9 = RigVMVariableNode() {variable=Auto Weights}
   n10 = RigVMDispatchNode(
    True=&RigVMFunction_MathFloatDiv(B=@n8.Result, A=@n6.Result) {template="Divide::Execute(in A,in B,out Result)"}.Result
    Condition=@n9.Value
    False=RigVMVariableNode() {variable=Weight 1}.Value
   ) {template="DISPATCH_RigVMDispatch_If(in Condition,in True,in False,out Result)"}
   b3t1 = RigVMCollapseNode(t=@n10.Result)
   n11 = RigVMVariableNode() {variable=Point 0}
   b0t1 = RigVMCollapseNode(t=@n10.Result)
   n12 = &RigVMFunction_MathVectorScale(Value=@n11.Value, Factor=@b0t1.Result) {template="Scale::Execute(in Value,in Factor,out Result)"}
   n13 = RigVMVariableNode() {variable=Point 1}
   n14 = RigVMDispatchNode(
    True=&RigVMFunction_MathFloatDiv(A=@n7.Result, B=@n8.Result) {template="Divide::Execute(in A,in B,out Result)"}.Result
    Condition=@n9.Value
    False=RigVMVariableNode() {variable=Weight 2}.Value
   ) {template="DISPATCH_RigVMDispatch_If(in Condition,in True,in False,out Result)"}
   b3t2 = RigVMCollapseNode(t=@n14.Result)
   b0t2 = RigVMCollapseNode(t=@n14.Result)
   n15 = &RigVMFunction_MathVectorScale(Value=@n11.Value, Factor=@b0t2.Result) {template="Scale::Execute(in Value,in Factor,out Result)"}
   n16 = RigVMVariableNode() {variable=Point 2}
   b1t1 = RigVMCollapseNode(t=@n10.Result)
   b1t2 = RigVMCollapseNode(t=@n14.Result)
   b2t1 = RigVMCollapseNode(t=@n10.Result)
   b2t2 = RigVMCollapseNode(t=@n14.Result)
   Solve = RigVMCollapseNode(
    c=/Script/RigVM.RigVMFunction_MathVectorSub(
     B=&RigVMFunction_MathVectorScale(Value=@n5.Value, Factor=@b3t1.Result) {template="Scale::Execute(in Value,in Factor,out Result)"}.Result
     A=&RigVMFunction_MathVectorSub(B=@n12.Result, A=@n13.Value) {template="Subtract::Execute(in A,in B,out Result)"}.Result
    ) {is=struct RigVMUnitNode, template="Subtract::Execute(in A,in B,out Result)"}.Result
    f=&RigVMFunction_MathVectorSub(
     B=&RigVMFunction_MathVectorScale(Value=@n5.Value, Factor=@b3t2.Result) {template="Scale::Execute(in Value,in Factor,out Result)"}.Result
     A=&RigVMFunction_MathVectorSub(B=@n15.Result, A=@n16.Value) {template="Subtract::Execute(in A,in B,out Result)"}.Result
    ) {template="Subtract::Execute(in A,in B,out Result)"}.Result
    a=@b1t1.Result
    d=@b1t2.Result
    b=@b2t1.Result
    e=@b2t2.Result
   )
   > n17 = RigVMDispatchNode(Index=1, Array=@n4.Array, Element=@Solve.i) {template="DISPATCH_RigVMDispatch_ArraySetAtIndex(io Array,in Index,in Element)"}
   > n18 = RigVMDispatchNode(Index=2, Array=@n17.Array, Element=@Solve.j) {template="DISPATCH_RigVMDispatch_ArraySetAtIndex(io Array,in Index,in Element)"}
   n19 = RigVMVariableNode() {variable=Point 3}
   > n20 = RigVMDispatchNode(Index=3, Array=@n18.Array, Element=@n19.Value) {template="DISPATCH_RigVMDispatch_ArraySetAtIndex(io Array,in Index,in Element)"}
   RigVMFunctionReturnNode(
    Control Points=@n20.Array
    Execute=@n20.ExecuteContext
    Spline=&RigUnit_ControlRigSplineFromPoints(
     Points=((X=0,Y=0,Z=0),(X=0,Y=0,Z=0),(X=0,Y=0,Z=0),(X=0,Y=0,Z=0))
     SamplesPerSegment=16
     SplineMode=BSpline
     Points.1=@Solve.i
     Points.2=@Solve.j
     Points.0=@n3.Value
     Points.3=@n19.Value
    ).Spline
   )
  ]}
 ContainedGraph {nodes [
   RigVMFunctionEntryNode()
   n1 = &RigVMFunction_MathFloatDiv(
    A=&RigVMFunction_MathFloatMul(A=RigVMVariableNode() {variable=a}.Value, B=RigVMVariableNode() {variable=e}.Value) {template="Multiply::Execute(in A,in B,out Result)"}.Result
    B=RigVMVariableNode() {variable=d}.Value
   ) {template="Divide::Execute(in A,in B,out Result)"}
   n2 = RigVMVariableNode() {variable=b}
   n3 = &RigVMFunction_MathFloatDiv(A=RigVMVariableNode() {variable=a}.Value, B=RigVMVariableNode() {variable=d}.Value) {template="Divide::Execute(in A,in B,out Result)"}
   n4 = RigVMVariableNode() {variable=f}
   n5 = &RigVMFunction_MathVectorScale(
    Factor=&RigVMFunction_MathFloatDiv(
     A=1
     B=/Script/RigVM.RigVMFunction_MathFloatSub(B=@n1.Result, A=@n2.Value) {is=struct RigVMUnitNode, template="Subtract::Execute(in A,in B,out Result)"}.Result
    ) {template="Divide::Execute(in A,in B,out Result)"}.Result
    Value=&RigVMFunction_MathVectorSub(
     B=&RigVMFunction_MathVectorScale(Factor=@n3.Result, Value=@n4.Value) {template="Scale::Execute(in Value,in Factor,out Result)"}.Result
     A=RigVMVariableNode() {variable=c}.Value
    ) {template="Subtract::Execute(in A,in B,out Result)"}.Result
   ) {template="Scale::Execute(in Value,in Factor,out Result)"}
   n6 = RigVMVariableNode() {variable=a}
   n7 = &RigVMFunction_MathVectorScale(Value=@n5.Result, Factor=RigVMVariableNode() {variable=b}.Value) {template="Scale::Execute(in Value,in Factor,out Result)"}
   n8 = RigVMVariableNode() {variable=c}
   RigVMFunctionReturnNode(
    j=@n5.Result
    i=&RigVMFunction_MathVectorScale(
     Factor=&RigVMFunction_MathFloatDiv(A=1, B=@n6.Value) {template="Divide::Execute(in A,in B,out Result)"}.Result
     Value=&RigVMFunction_MathVectorSub(B=@n7.Result, A=@n8.Value) {template="Subtract::Execute(in A,in B,out Result)"}.Result
    ) {template="Scale::Execute(in Value,in Factor,out Result)"}.Result
   )
  ]}
 ContainedGraph {nodes [
   n1 = RigVMFunctionEntryNode()
   n2 = &RigVMFunction_MathFloatSub(A=1, B=@n1.t) {template="Subtract::Execute(in A,in B,out Result)"}
   n3 = &RigVMFunction_MathFloatMul(A=@n2.Result, B=@n2.Result) {template="Multiply::Execute(in A,in B,out Result)"}
   RigVMFunctionReturnNode(Result=&RigVMFunction_MathFloatMul(B=3, A=&RigVMFunction_MathFloatMul(A=@n1.t, B=@n3.Result) {template="Multiply::Execute(in A,in B,out Result)"}.Result) {template="Multiply::Execute(in A,in B,out Result)"}.Result)
  ]}
 ContainedGraph {nodes [
   n1 = RigVMFunctionEntryNode()
   n2 = &RigVMFunction_MathFloatMul(A=@n1.t, B=@n1.t) {template="Multiply::Execute(in A,in B,out Result)"}
   n3 = &RigVMFunction_MathFloatSub(A=1, B=@n1.t) {template="Subtract::Execute(in A,in B,out Result)"}
   RigVMFunctionReturnNode(Result=&RigVMFunction_MathFloatMul(B=3, A=&RigVMFunction_MathFloatMul(B=@n2.Result, A=@n3.Result) {template="Multiply::Execute(in A,in B,out Result)"}.Result) {template="Multiply::Execute(in A,in B,out Result)"}.Result)
  ]}
 ContainedGraph {nodes [
   n1 = RigVMFunctionEntryNode()
   n2 = &RigVMFunction_MathFloatSub(A=1, B=@n1.t) {template="Subtract::Execute(in A,in B,out Result)"}
   n3 = &RigVMFunction_MathFloatMul(A=@n2.Result, B=@n2.Result) {template="Multiply::Execute(in A,in B,out Result)"}
   RigVMFunctionReturnNode(Result=&RigVMFunction_MathFloatMul(B=3, A=&RigVMFunction_MathFloatMul(A=@n1.t, B=@n3.Result) {template="Multiply::Execute(in A,in B,out Result)"}.Result) {template="Multiply::Execute(in A,in B,out Result)"}.Result)
  ]}
 ContainedGraph {nodes [
   n1 = RigVMFunctionEntryNode()
   n2 = &RigVMFunction_MathFloatMul(A=@n1.t, B=@n1.t) {template="Multiply::Execute(in A,in B,out Result)"}
   n3 = &RigVMFunction_MathFloatSub(A=1, B=@n1.t) {template="Subtract::Execute(in A,in B,out Result)"}
   RigVMFunctionReturnNode(Result=&RigVMFunction_MathFloatMul(B=3, A=&RigVMFunction_MathFloatMul(B=@n2.Result, A=@n3.Result) {template="Multiply::Execute(in A,in B,out Result)"}.Result) {template="Multiply::Execute(in A,in B,out Result)"}.Result)
  ]}
 ContainedGraph {nodes [
   n1 = RigVMFunctionEntryNode()
   RigVMFunctionReturnNode(Result=/Script/RigVM.RigVMFunction_MathFloatPow(B=3, A=&RigVMFunction_MathFloatSub(A=1, B=@n1.t) {template="Subtract::Execute(in A,in B,out Result)"}.Result) {is=struct RigVMUnitNode, template="Power::Execute(in A,in B,out Result)"}.Result)
  ]}
 ContainedGraph {nodes [RigVMFunctionReturnNode(Result=&RigVMFunction_MathFloatPow(B=3, A=RigVMFunctionEntryNode().t) {template="Power::Execute(in A,in B,out Result)"}.Result)]}
 ContainedGraph {nodes [
   n1 = RigVMFunctionEntryNode()
   RigVMFunctionReturnNode(Result=&RigVMFunction_MathFloatPow(B=3, A=&RigVMFunction_MathFloatSub(A=1, B=@n1.t) {template="Subtract::Execute(in A,in B,out Result)"}.Result) {template="Power::Execute(in A,in B,out Result)"}.Result)
  ]}
 ContainedGraph {nodes [RigVMFunctionReturnNode(Result=&RigVMFunction_MathFloatPow(B=3, A=RigVMFunctionEntryNode().t) {template="Power::Execute(in A,in B,out Result)"}.Result)]}
 ContainedGraph {nodes [
   n1 = RigVMFunctionEntryNode()
   > SequenceExecution = RigVMAggregateNode()
   n2 = RigVMDispatchNode(ExecuteContext=@SequenceExecution.A, Array=RigVMVariableNode() {variable=FK Controls}.Value) {template="DISPATCH_RigVMDispatch_ArrayIterator(in Array,out Element,out Index,out Count,out Ratio)"}
   > /Script/ControlRig.RigUnit_SetControlVisibility(
    Item=@n2.Element
    bVisible=RigVMDispatchNode(False=False, Condition=@n1.Show Controls, True=&RigVMFunction_MathBoolNot(Value=@n1.IKFK Switch).Result) {template="DISPATCH_RigVMDispatch_If(in Condition,in True,in False,out Result)"}.Result
   ) {is=struct RigVMUnitNode, template="SetControlVisibility::Execute(in Item,in Pattern,in bVisible)"}
   n3 = RigVMDispatchNode(ExecuteContext=@SequenceExecution.B, Array=RigVMVariableNode() {variable=IK Controls}.Value) {template="DISPATCH_RigVMDispatch_ArrayIterator(in Array,out Element,out Index,out Count,out Ratio)"}
   > &RigUnit_SetControlVisibility(
    Item=@n3.Element
    bVisible=RigVMDispatchNode(False=False, True=@n1.IKFK Switch, Condition=@n1.Show Controls) {template="DISPATCH_RigVMDispatch_If(in Condition,in True,in False,out Result)"}.Result
   ) {template="SetControlVisibility::Execute(in Item,in Pattern,in bVisible)"}
   RigVMFunctionReturnNode(ExecuteContext=@SequenceExecution.D)
  ]}
 ContainedGraph {nodes [
   n1 = RigVMFunctionEntryNode()
   SequenceExecution = &RigVMFunction_Sequence(ExecuteContext=@n1.ExecuteContext)
   SequenceExecution_1 = &RigVMFunction_Sequence(ExecuteContext=@SequenceExecution.B)
   SequenceExecution_2 = &RigVMFunction_Sequence(ExecuteContext=@SequenceExecution_1.B)
   RigVMFunctionReturnNode(A=@SequenceExecution.A, B=@SequenceExecution_1.A, C=@SequenceExecution_2.A, D=@SequenceExecution_2.B)
  ]}
 ContainedGraph {nodes [
   n1 = RigVMFunctionEntryNode()
   > n2 = RigVMDispatchNode(Array=@n1.Children) {template="DISPATCH_RigVMDispatch_ArrayIterator(in Array,out Element,out Index,out Count,out Ratio)"}
   n3 = RigVMDispatchNode(Index=@n2.Index, Array=@n1.Parents) {template="DISPATCH_RigVMDispatch_ArrayGetAtIndex(in Array,in Index,out Element)"}
   > &RigUnit_SetTransform(
    Weight=1
    Item=@n2.Element
    bPropagateToChildren=@n1.Propagate to Children
    Value=&RigUnit_ProjectTransformToNewParent(bChildInitial=True, bOldParentInitial=True, NewParent=@n3.Element, OldParent=@n3.Element, Child=@n2.Element).Transform
   )
   RigVMFunctionReturnNode(ExecuteContext=@n2.Completed)
  ]}
 ContainedGraph {nodes [
   RigVMFunctionEntryNode()
   > SequenceExecution = RigVMAggregateNode()
   RigVMFunctionReturnNode(ExecuteContext=@SequenceExecution.D)
   n1 = RigVMDispatchNode(Array=RigVMVariableNode() {variable=Bones}.Value, ExecuteContext=@SequenceExecution.A) {template="DISPATCH_RigVMDispatch_ArrayIterator(in Array,out Element,out Index,out Count,out Ratio)"}
   n2 = &RigVMFunction_ControlFlowBranch(
    ExecuteContext=@n1.ExecuteContext
    Condition=/Script/RigVM.RigVMFunction_MathIntGreaterEqual(A=@n1.Index, B=&RigVMFunction_MathIntSub(B=1, A=@n1.Count) {template="Subtract::Execute(in A,in B,out Result)"}.Result) {is=struct RigVMUnitNode, template="GreaterEqual::Execute(in A,in B,out Result)"}.Result
   )
   n3 = RigVMVariableNode() {variable=FK Controls}
   n4 = &RigUnit_AimBoneMath(
    DebugSettings=(Scale=10,WorldOffset=(Rotation=(X=0,Y=0,Z=0,W=1),Translation=(X=0,Y=0,Z=0),Scale3D=(X=1,Y=1,Z=1)))
    InputTransform=(Rotation=(X=0,Y=0,Z=0,W=1),Translation=(X=0,Y=0,Z=0),Scale3D=(X=1,Y=1,Z=1))
    Primary=(Weight=1,Axis=(X=1,Y=0,Z=0),Target=(X=0,Y=0,Z=0),Kind=Location,Space=(Type=Bone))
    Secondary=(Weight=1,Axis=(X=0,Y=0,Z=1),Target=(X=1,Y=0,Z=0),Kind=Direction,Space=(Type=Bone))
    Weight=1
    Secondary.Space=RigVMDispatchNode(Index=@n1.Index, Array=@n3.Value) {template="DISPATCH_RigVMDispatch_ArrayGetAtIndex(in Array,in Index,out Element)"}.Element
    Primary.Space=RigVMDispatchNode(Index=&RigVMFunction_MathIntAdd(B=1, A=@n1.Index) {template="Add::Execute(in A,in B,out Result)"}.Result, Array=@n3.Value) {template="DISPATCH_RigVMDispatch_ArrayGetAtIndex(in Array,in Index,out Element)"}.Element
    InputTransform.Translation=RigVMDispatchNode(Index=@n1.Index, Array=&RigUnit_GetTransformItemArray(Space=GlobalSpace, Items=@n3.Value).Transforms) {template="DISPATCH_RigVMDispatch_ArrayGetAtIndex(in Array,in Index,out Element)"}.Element.Translation
    Primary.Axis=RigVMVariableNode() {variable=Point At Axis}.Value
    Secondary.Axis=RigVMVariableNode() {variable=Up Vector Axis}.Value
    Secondary.Target=RigVMVariableNode() {variable=Up Vector Target}.Value
   )
   &RigUnit_SetTransform(
    Weight=1
    Item=@n1.Element
    ExecutePin=@n2.False
    Value=RigVMDispatchNode(
     True=(Rotation=(X=0,Y=0,Z=0,W=1),Translation=(X=0,Y=0,Z=0),Scale3D=(X=0,Y=0,Z=0))
     False=@n4.Result
     Condition=RigVMVariableNode() {variable=Negative Side}.Value
    ) {template="DISPATCH_RigVMDispatch_If(in Condition,in True,in False,out Result)"}.Result
   )
   n5 = RigVMVariableNode() {variable=Bones}
   n6 = &RigVMFunction_MathIntSub(B=1, A=RigVMDispatchNode(Array=@n5.Value) {template="DISPATCH_RigVMDispatch_ArrayGetNum(in Array,out Num)"}.Num) {template="Subtract::Execute(in A,in B,out Result)"}
   n7 = RigVMDispatchNode(Index=@n6.Result, Array=@n5.Value) {template="DISPATCH_RigVMDispatch_ArrayGetAtIndex(in Array,in Index,out Element)"}
   n8 = RigVMDispatchNode(Index=@n6.Result, Array=RigVMVariableNode() {variable=FK Controls}.Value) {template="DISPATCH_RigVMDispatch_ArrayGetAtIndex(in Array,in Index,out Element)"}
   &RigUnit_SetTransform(
    Weight=1
    Item=@n7.Element
    Value=&RigUnit_ProjectTransformToNewParent(bChildInitial=True, bOldParentInitial=True, Child=@n7.Element, NewParent=@n8.Element, OldParent=@n8.Element).Transform
    ExecutePin=@SequenceExecution.B
   )
  ]}
 ContainedGraph {nodes [
   n1 = RigVMFunctionEntryNode()
   SequenceExecution = &RigVMFunction_Sequence(ExecuteContext=@n1.ExecuteContext)
   SequenceExecution_1 = &RigVMFunction_Sequence(ExecuteContext=@SequenceExecution.B)
   SequenceExecution_2 = &RigVMFunction_Sequence(ExecuteContext=@SequenceExecution_1.B)
   RigVMFunctionReturnNode(A=@SequenceExecution.A, B=@SequenceExecution_1.A, C=@SequenceExecution_2.A, D=@SequenceExecution_2.B)
  ]}
 ContainedGraph {nodes [
   n1 = RigVMFunctionEntryNode()
   n2 = RigVMDispatchNode(Num=4, Array=RigVMVariableNode() {variable=Outputs}.Value, ExecuteContext=@n1.Execute) {template="DISPATCH_RigVMDispatch_ArraySetNum(io Array,in Num)"}
   n3 = RigVMVariableNode() {variable=Point 0}
   > n4 = RigVMDispatchNode(Index=0, Array=@n2.Array, Element=@n3.Value) {template="DISPATCH_RigVMDispatch_ArraySetAtIndex(io Array,in Index,in Element)"}
   n5 = RigVMVariableNode() {variable=Point 3}
   n6 = &RigVMFunction_MathVectorDistance(B=@n1.Point 0, A=@n1.Point 1)
   n7 = &RigVMFunction_MathFloatAdd(A=@n6.Result, B=&RigVMFunction_MathVectorDistance(A=@n1.Point 1, B=@n1.Point 2).Result) {template="Add::Execute(in A,in B,out Result)"}
   n8 = &RigVMFunction_MathFloatAdd(A=@n7.Result, B=&RigVMFunction_MathVectorDistance(A=@n1.Point 2, B=@n1.Point 3).Result) {template="Add::Execute(in A,in B,out Result)"}
   n9 = RigVMVariableNode() {variable=Auto Weights}
   n10 = RigVMDispatchNode(
    True=&RigVMFunction_MathFloatDiv(B=@n8.Result, A=@n6.Result) {template="Divide::Execute(in A,in B,out Result)"}.Result
    Condition=@n9.Value
    False=RigVMVariableNode() {variable=Weight 1}.Value
   ) {template="DISPATCH_RigVMDispatch_If(in Condition,in True,in False,out Result)"}
   b3t1 = RigVMCollapseNode(t=@n10.Result)
   n11 = RigVMVariableNode() {variable=Point 0}
   b0t1 = RigVMCollapseNode(t=@n10.Result)
   n12 = &RigVMFunction_MathVectorScale(Value=@n11.Value, Factor=@b0t1.Result) {template="Scale::Execute(in Value,in Factor,out Result)"}
   n13 = RigVMVariableNode() {variable=Point 1}
   n14 = RigVMDispatchNode(
    True=&RigVMFunction_MathFloatDiv(A=@n7.Result, B=@n8.Result) {template="Divide::Execute(in A,in B,out Result)"}.Result
    Condition=@n9.Value
    False=RigVMVariableNode() {variable=Weight 2}.Value
   ) {template="DISPATCH_RigVMDispatch_If(in Condition,in True,in False,out Result)"}
   b3t2 = RigVMCollapseNode(t=@n14.Result)
   b0t2 = RigVMCollapseNode(t=@n14.Result)
   n15 = &RigVMFunction_MathVectorScale(Value=@n11.Value, Factor=@b0t2.Result) {template="Scale::Execute(in Value,in Factor,out Result)"}
   n16 = RigVMVariableNode() {variable=Point 2}
   b1t1 = RigVMCollapseNode(t=@n10.Result)
   b1t2 = RigVMCollapseNode(t=@n14.Result)
   b2t1 = RigVMCollapseNode(t=@n10.Result)
   b2t2 = RigVMCollapseNode(t=@n14.Result)
   Solve = RigVMCollapseNode(
    c=&RigVMFunction_MathVectorSub(
     B=&RigVMFunction_MathVectorScale(Value=@n5.Value, Factor=@b3t1.Result) {template="Scale::Execute(in Value,in Factor,out Result)"}.Result
     A=&RigVMFunction_MathVectorSub(B=@n12.Result, A=@n13.Value) {template="Subtract::Execute(in A,in B,out Result)"}.Result
    ) {template="Subtract::Execute(in A,in B,out Result)"}.Result
    f=&RigVMFunction_MathVectorSub(
     B=&RigVMFunction_MathVectorScale(Value=@n5.Value, Factor=@b3t2.Result) {template="Scale::Execute(in Value,in Factor,out Result)"}.Result
     A=&RigVMFunction_MathVectorSub(B=@n15.Result, A=@n16.Value) {template="Subtract::Execute(in A,in B,out Result)"}.Result
    ) {template="Subtract::Execute(in A,in B,out Result)"}.Result
    a=@b1t1.Result
    d=@b1t2.Result
    b=@b2t1.Result
    e=@b2t2.Result
   )
   > n17 = RigVMDispatchNode(Index=1, Array=@n4.Array, Element=@Solve.i) {template="DISPATCH_RigVMDispatch_ArraySetAtIndex(io Array,in Index,in Element)"}
   > n18 = RigVMDispatchNode(Index=2, Array=@n17.Array, Element=@Solve.j) {template="DISPATCH_RigVMDispatch_ArraySetAtIndex(io Array,in Index,in Element)"}
   n19 = RigVMVariableNode() {variable=Point 3}
   > n20 = RigVMDispatchNode(Index=3, Array=@n18.Array, Element=@n19.Value) {template="DISPATCH_RigVMDispatch_ArraySetAtIndex(io Array,in Index,in Element)"}
   RigVMFunctionReturnNode(
    Control Points=@n20.Array
    Execute=@n20.ExecuteContext
    Spline=&RigUnit_ControlRigSplineFromPoints(
     Points=((X=0,Y=0,Z=0),(X=0,Y=0,Z=0),(X=0,Y=0,Z=0),(X=0,Y=0,Z=0))
     SamplesPerSegment=16
     SplineMode=BSpline
     Points.1=@Solve.i
     Points.2=@Solve.j
     Points.0=@n3.Value
     Points.3=@n19.Value
    ).Spline
   )
  ]}
 ContainedGraph {nodes [
   RigVMFunctionEntryNode()
   n1 = &RigVMFunction_MathFloatDiv(
    A=&RigVMFunction_MathFloatMul(A=RigVMVariableNode() {variable=a}.Value, B=RigVMVariableNode() {variable=e}.Value) {template="Multiply::Execute(in A,in B,out Result)"}.Result
    B=RigVMVariableNode() {variable=d}.Value
   ) {template="Divide::Execute(in A,in B,out Result)"}
   n2 = RigVMVariableNode() {variable=b}
   n3 = &RigVMFunction_MathFloatDiv(A=RigVMVariableNode() {variable=a}.Value, B=RigVMVariableNode() {variable=d}.Value) {template="Divide::Execute(in A,in B,out Result)"}
   n4 = RigVMVariableNode() {variable=f}
   n5 = &RigVMFunction_MathVectorScale(
    Factor=&RigVMFunction_MathFloatDiv(A=1, B=&RigVMFunction_MathFloatSub(B=@n1.Result, A=@n2.Value) {template="Subtract::Execute(in A,in B,out Result)"}.Result) {template="Divide::Execute(in A,in B,out Result)"}.Result
    Value=&RigVMFunction_MathVectorSub(
     B=&RigVMFunction_MathVectorScale(Factor=@n3.Result, Value=@n4.Value) {template="Scale::Execute(in Value,in Factor,out Result)"}.Result
     A=RigVMVariableNode() {variable=c}.Value
    ) {template="Subtract::Execute(in A,in B,out Result)"}.Result
   ) {template="Scale::Execute(in Value,in Factor,out Result)"}
   n6 = RigVMVariableNode() {variable=a}
   n7 = &RigVMFunction_MathVectorScale(Value=@n5.Result, Factor=RigVMVariableNode() {variable=b}.Value) {template="Scale::Execute(in Value,in Factor,out Result)"}
   n8 = RigVMVariableNode() {variable=c}
   RigVMFunctionReturnNode(
    j=@n5.Result
    i=&RigVMFunction_MathVectorScale(
     Factor=&RigVMFunction_MathFloatDiv(A=1, B=@n6.Value) {template="Divide::Execute(in A,in B,out Result)"}.Result
     Value=&RigVMFunction_MathVectorSub(B=@n7.Result, A=@n8.Value) {template="Subtract::Execute(in A,in B,out Result)"}.Result
    ) {template="Scale::Execute(in Value,in Factor,out Result)"}.Result
   )
  ]}
 ContainedGraph {nodes [
   n1 = RigVMFunctionEntryNode()
   n2 = &RigVMFunction_MathFloatSub(A=1, B=@n1.t) {template="Subtract::Execute(in A,in B,out Result)"}
   n3 = &RigVMFunction_MathFloatMul(A=@n2.Result, B=@n2.Result) {template="Multiply::Execute(in A,in B,out Result)"}
   RigVMFunctionReturnNode(Result=&RigVMFunction_MathFloatMul(B=3, A=&RigVMFunction_MathFloatMul(A=@n1.t, B=@n3.Result) {template="Multiply::Execute(in A,in B,out Result)"}.Result) {template="Multiply::Execute(in A,in B,out Result)"}.Result)
  ]}
 ContainedGraph {nodes [
   n1 = RigVMFunctionEntryNode()
   n2 = &RigVMFunction_MathFloatMul(A=@n1.t, B=@n1.t) {template="Multiply::Execute(in A,in B,out Result)"}
   n3 = &RigVMFunction_MathFloatSub(A=1, B=@n1.t) {template="Subtract::Execute(in A,in B,out Result)"}
   RigVMFunctionReturnNode(Result=&RigVMFunction_MathFloatMul(B=3, A=&RigVMFunction_MathFloatMul(B=@n2.Result, A=@n3.Result) {template="Multiply::Execute(in A,in B,out Result)"}.Result) {template="Multiply::Execute(in A,in B,out Result)"}.Result)
  ]}
 ContainedGraph {nodes [
   n1 = RigVMFunctionEntryNode()
   n2 = &RigVMFunction_MathFloatSub(A=1, B=@n1.t) {template="Subtract::Execute(in A,in B,out Result)"}
   n3 = &RigVMFunction_MathFloatMul(A=@n2.Result, B=@n2.Result) {template="Multiply::Execute(in A,in B,out Result)"}
   RigVMFunctionReturnNode(Result=&RigVMFunction_MathFloatMul(B=3, A=&RigVMFunction_MathFloatMul(A=@n1.t, B=@n3.Result) {template="Multiply::Execute(in A,in B,out Result)"}.Result) {template="Multiply::Execute(in A,in B,out Result)"}.Result)
  ]}
 ContainedGraph {nodes [
   n1 = RigVMFunctionEntryNode()
   n2 = &RigVMFunction_MathFloatMul(A=@n1.t, B=@n1.t) {template="Multiply::Execute(in A,in B,out Result)"}
   n3 = &RigVMFunction_MathFloatSub(A=1, B=@n1.t) {template="Subtract::Execute(in A,in B,out Result)"}
   RigVMFunctionReturnNode(Result=&RigVMFunction_MathFloatMul(B=3, A=&RigVMFunction_MathFloatMul(B=@n2.Result, A=@n3.Result) {template="Multiply::Execute(in A,in B,out Result)"}.Result) {template="Multiply::Execute(in A,in B,out Result)"}.Result)
  ]}
 ContainedGraph {nodes [
   n1 = RigVMFunctionEntryNode()
   RigVMFunctionReturnNode(Result=&RigVMFunction_MathFloatPow(B=3, A=&RigVMFunction_MathFloatSub(A=1, B=@n1.t) {template="Subtract::Execute(in A,in B,out Result)"}.Result) {template="Power::Execute(in A,in B,out Result)"}.Result)
  ]}
 ContainedGraph {nodes [RigVMFunctionReturnNode(Result=&RigVMFunction_MathFloatPow(B=3, A=RigVMFunctionEntryNode().t) {template="Power::Execute(in A,in B,out Result)"}.Result)]}
 ContainedGraph {nodes [
   n1 = RigVMFunctionEntryNode()
   RigVMFunctionReturnNode(Result=&RigVMFunction_MathFloatPow(B=3, A=&RigVMFunction_MathFloatSub(A=1, B=@n1.t) {template="Subtract::Execute(in A,in B,out Result)"}.Result) {template="Power::Execute(in A,in B,out Result)"}.Result)
  ]}
 ContainedGraph {nodes [RigVMFunctionReturnNode(Result=&RigVMFunction_MathFloatPow(B=3, A=RigVMFunctionEntryNode().t) {template="Power::Execute(in A,in B,out Result)"}.Result)]}
 ContainedGraph {nodes [
   RigVMFunction_MathTransformArrayToSRT = &RigVMFunction_MathTransformArrayToSRT(Transforms=&RigUnit_GetTransformItemArray(Space=GlobalSpace, bInitial=False, Items=RigVMVariableNode() {variable=Bones}.Value).Transforms)
   SequenceExecution = RigVMAggregateNode(ExecuteContext=@n6.Execute)
   BSpline Passing thru 4 Points_1_1 = RigVMFunctionReferenceNode(
    Auto Weights=False
    Weight 1=0.333
    Weight 2=0.666
    Point 0=RigVMDispatchNode(Index=0, Array=@RigVMFunction_MathTransformArrayToSRT.Translations) {template="DISPATCH_RigVMDispatch_ArrayGetAtIndex(in Array,in Index,out Element)"}.Element
    Point 1=RigVMDispatchNode(Index=1, Array=@RigVMFunction_MathTransformArrayToSRT.Translations) {template="DISPATCH_RigVMDispatch_ArrayGetAtIndex(in Array,in Index,out Element)"}.Element
    Point 2=RigVMDispatchNode(Index=2, Array=@RigVMFunction_MathTransformArrayToSRT.Translations) {template="DISPATCH_RigVMDispatch_ArrayGetAtIndex(in Array,in Index,out Element)"}.Element
    Point 3=RigVMDispatchNode(Index=3, Array=@RigVMFunction_MathTransformArrayToSRT.Translations) {template="DISPATCH_RigVMDispatch_ArrayGetAtIndex(in Array,in Index,out Element)"}.Element
    Execute=@SequenceExecution.A
   )
   n1 = RigVMDispatchNode(ExecuteContext=@"BSpline Passing thru 4 Points_1_1".Execute, Array=RigVMVariableNode() {variable=IK Controls}.Value) {template="DISPATCH_RigVMDispatch_ArrayIterator(in Array,out Element,out Index,out Count,out Ratio)"}
   MathIntEquals_1 = RigVMDispatchNode(B=0, A=@n1.Index) {template="DISPATCH_RigVMDispatch_CoreEquals(in A,in B,out Result)"}
   n2 = &RigVMFunction_ControlFlowBranch(ExecuteContext=@n1.ExecuteContext, Condition=@MathIntEquals_1.Result)
   n3 = RigVMDispatchNode(Index=@n1.Index, Array=RigVMVariableNode() {variable=Bones}.Value) {template="DISPATCH_RigVMDispatch_ArrayGetAtIndex(in Array,in Index,out Element)"}
   n4 = &RigUnit_ProjectTransformToNewParent(bChildInitial=True, bOldParentInitial=True, NewParent=@n3.Element, OldParent=@n3.Element, Child=@n1.Element)
   n5 = &RigVMFunction_MathTransformMakeAbsolute(
    Local=(Rotation=(X=0,Y=0,Z=0,W=1),Translation=(X=0,Y=0,Z=0),Scale3D=(X=1,Y=1,Z=1))
    Parent=(Rotation=(X=0,Y=0,Z=0,W=1),Translation=(X=0,Y=0,Z=0),Scale3D=(X=1,Y=1,Z=1))
    Parent.Translation=RigVMDispatchNode(Index=@n1.Index, Array=@"BSpline Passing thru 4 Points_1_1".Control Points) {template="DISPATCH_RigVMDispatch_ArrayGetAtIndex(in Array,in Index,out Element)"}.Element
    Parent.Rotation=@n4.Transform.Rotation
    Parent.Scale3D=@n4.Transform.Scale3D
   ) {template="Make Absolute::Execute(in Local,in Parent,out Global)"}
   &RigUnit_SetTransform(Weight=1, bPropagateToChildren=false, Item=@n1.Element, ExecutePin=@n2.False, Value=@n5.Global)
   n6 = RigVMFunctionEntryNode()
   RigVMDispatchNode(Index=@n1.Index, Array=RigVMVariableNode() {variable=Offsets}.Value) {template="DISPATCH_RigVMDispatch_ArrayGetAtIndex(in Array,in Index,out Element)"}
   RigVMFunctionReturnNode(Execute=@SequenceExecution.D)
  ]}
 ContainedGraph {nodes [
   n1 = RigVMFunctionEntryNode()
   SequenceExecution = &RigVMFunction_Sequence(ExecuteContext=@n1.ExecuteContext)
   SequenceExecution_1 = &RigVMFunction_Sequence(ExecuteContext=@SequenceExecution.B)
   SequenceExecution_2 = &RigVMFunction_Sequence(ExecuteContext=@SequenceExecution_1.B)
   RigVMFunctionReturnNode(A=@SequenceExecution.A, B=@SequenceExecution_1.A, C=@SequenceExecution_2.A, D=@SequenceExecution_2.B)
  ]}
 RigVMFunctionLibrary {nodes [
   BSpline Passing thru 4 Points = RigVMCollapseNode(
    Auto Weights=true
    Execute=()
    Point 0=(X=0,Y=0,Z=0)
    Point 1=(X=0,Y=0,Z=0)
    Point 2=(X=0,Y=0,Z=0)
    Point 3=(X=0,Y=0,Z=0)
    Weight 1=0.33
    Weight 2=0.66
   )
   BSpline Passing thru 4 Points_1 = RigVMCollapseNode(
    Auto Weights=true
    Execute=()
    Point 0=(X=0,Y=0,Z=0)
    Point 1=(X=0,Y=0,Z=0)
    Point 2=(X=0,Y=0,Z=0)
    Point 3=(X=0,Y=0,Z=0)
    Weight 1=0.33
    Weight 2=0.66
   )
   FK Point At = RigVMCollapseNode(Bones=(), ExecuteContext=(), FK Controls=(), Point At Axis=(X=1,Y=0,Z=0), Up Vector Axis=(X=0,Y=0,Z=1), Up Vector Target=(X=0,Y=0,Z=1))
   Project to New Parents = RigVMCollapseNode(Children=(), ExecuteContext=(), Parents=(), Propagate to Children=True)
   Set IKFK Visibility = RigVMCollapseNode(ExecuteContext=(), FK Controls=(), IK Controls=(), IKFK Switch=True)
   Spine Match IK onto FK = RigVMCollapseNode(Bones=(), Execute=(), IK Controls=(), Offsets=())
  ]}
 RigVMFunction_Sequence_1_ContainedGraph {nodes [
   n1 = RigVMFunctionEntryNode()
   RigVMFunction_Sequence = &RigVMFunction_Sequence(ExecuteContext=@n1.ExecuteContext)
   n2 = &RigVMFunction_Sequence(ExecuteContext=@RigVMFunction_Sequence.B)
   n3 = &RigVMFunction_Sequence(ExecuteContext=@n2.B)
   RigVMFunctionReturnNode(A=@RigVMFunction_Sequence.A, B=@n2.A, C=@n3.A, D=@n3.B)
  ]}
 RigVMFunction_Sequence_1_ContainedGraph {nodes [
   n1 = RigVMFunctionEntryNode()
   RigVMFunction_Sequence = &RigVMFunction_Sequence(ExecuteContext=@n1.ExecuteContext)
   n2 = &RigVMFunction_Sequence(ExecuteContext=@RigVMFunction_Sequence.B)
   n3 = &RigVMFunction_Sequence(ExecuteContext=@n2.B)
   n4 = &RigVMFunction_Sequence(ExecuteContext=@n3.B)
   RigVMFunctionReturnNode(A=@RigVMFunction_Sequence.A, B=@n2.A, C=@n3.A, D=@n4.A, E=@n4.B)
  ]}
 RigVMFunction_Sequence_1_ContainedGraph {nodes [
   n1 = RigVMFunctionEntryNode()
   RigVMFunction_Sequence = &RigVMFunction_Sequence(ExecuteContext=@n1.ExecuteContext)
   n2 = &RigVMFunction_Sequence(ExecuteContext=@RigVMFunction_Sequence.B)
   RigVMFunctionReturnNode(A=@RigVMFunction_Sequence.A, B=@n2.A, C=@n2.B)
  ]}
 RigVMModel {nodes [
   RigUnit_BeginExecution = &RigUnit_BeginExecution()
   n1 = RigVMVariableNode() {variable=IK Channel}
   RigVMVariableNode(
    Value=/Script/RigVM.RigVMFunction_MathFloatGreater(B=0.5, A=&RigUnit_GetTransform(Space=LocalSpace, Item=@n1.Value).Transform.Translation.X) {is=struct RigVMUnitNode, template="Greater::Execute(in A,in B,out Result)"}.Result
    ExecuteContext=@RigUnit_BeginExecution.ExecutePin
   ) {variable=IK Mode}
   n2 = RigVMVariableNode() {variable=Stretch Channel}
   > RigVMVariableNode(Value=&RigVMFunction_MathFloatGreater(B=0.5, A=&RigUnit_GetTransform(Space=LocalSpace, Item=@n2.Value).Transform.Translation.X) {template="Greater::Execute(in A,in B,out Result)"}.Result) {variable=Stretch}
   > RigVMFunction_Sequence = RigVMAggregateNode()
   n3 = &RigVMFunction_ControlFlowBranch(ExecuteContext=@RigVMFunction_Sequence.C, Condition=RigVMVariableNode() {variable=IK Mode}.Value)
   Clone = RigVMDispatchNode(Array=RigVMVariableNode() {variable=FK Controls}.Value) {template="DISPATCH_RigVMDispatch_ArrayClone(in Array,out Clone)"}
   n4 = RigVMDispatchNode(Index=0, Array=@Clone.Clone, Element=RigVMVariableNode() {variable=Hip Control}.Value, ExecuteContext=@n3.False) {template="DISPATCH_RigVMDispatch_ArrayInsert(io Array,in Index,in Element)"}
   > FK Point At_1 = RigVMFunctionReferenceNode(
    Point At Axis=(X=1,Y=0,Z=0)
    Up Vector Axis=(X=0,Y=0,Z=-1)
    Up Vector Target=(X=1,Y=0,Z=0)
    FK Controls=@n4.Array
    Bones=RigVMVariableNode() {variable=Bones}.Value
   )
   Spine Match IK onto FK = RigVMFunctionReferenceNode(
    Execute=@"FK Point At_1".ExecuteContext
    Offsets=RigVMVariableNode() {variable=IK Offsets}.Value
    Bones=RigVMVariableNode() {variable=Bones}.Value
    IK Controls=RigVMVariableNode() {variable=IK Controls}.Value
   )
   n5 = RigVMDispatchNode(
    Index=0
    Array=RigVMDispatchNode(Array=RigVMVariableNode() {variable=IK Controls}.Value) {template="DISPATCH_RigVMDispatch_ArrayClone(in Array,out Clone)"}.Clone
    ExecuteContext=@RigVMFunction_Sequence.A
   ) {template="DISPATCH_RigVMDispatch_ArrayRemove(io Array,in Index)"}
   > Set IKFK Visibility = RigVMFunctionReferenceNode(
    Show Controls=true
    IK Controls=@n5.Array
    IKFK Switch=RigVMVariableNode() {variable=IK Mode}.Value
    FK Controls=RigVMVariableNode() {variable=FK Controls}.Value
   )
   n6 = RigVMVariableNode() {variable=Hip Control}
   n7 = RigVMVariableNode() {variable=Pelvis Bone}
   &RigUnit_SetTransform(
    Weight=1
    Value=&RigUnit_ProjectTransformToNewParent(bChildInitial=True, bOldParentInitial=True, NewParent=@n6.Value, OldParent=@n6.Value, Child=@n7.Value).Transform
    ExecutePin=@RigVMFunction_Sequence.B
    Item=@n7.Value
   )
   n8 = RigVMDispatchNode(Index=-1, Array=RigVMVariableNode() {variable=Bones}.Value) {template="DISPATCH_RigVMDispatch_ArrayGetAtIndex(in Array,in Index,out Element)"}
   &RigUnit_SetTransform(
    Weight=1
    Value=&RigUnit_ProjectTransformToNewParent(Child=(Type=Null,Name="Chest Space"), bChildInitial=True, bOldParentInitial=True, NewParent=@n8.Element, OldParent=@n8.Element).Transform
    Item=RigVMVariableNode() {variable=Chest Space Null}.Value
    ExecutePin=@RigVMFunction_Sequence.D
   )
   Spine IK Solve = RigVMCollapseNode(
    Enable Stretch=RigVMVariableNode() {variable=Stretch}.Value
    Bones=RigVMVariableNode() {variable=Bones}.Value
    IK Controls=RigVMVariableNode() {variable=IK Controls}.Value
    Spine Offsets=RigVMVariableNode() {variable=IK Offsets}.Value
    Default Length=RigVMVariableNode() {variable=Length}.Value
    Completed=@n3.True
   )
   n9 = RigVMDispatchNode(
    Index=0
    Array=RigVMDispatchNode(Array=RigVMVariableNode() {variable=Bones}.Value) {template="DISPATCH_RigVMDispatch_ArrayClone(in Array,out Clone)"}.Clone
    ExecuteContext=@"Spine IK Solve".Completed
   ) {template="DISPATCH_RigVMDispatch_ArrayRemove(io Array,in Index)"}
   > Project to New Parent = RigVMFunctionReferenceNode(Propagate to Children=True, Parents=@n9.Array, Children=RigVMVariableNode() {variable=FK Controls}.Value)
  ]}
 RigVMModel Backwards Solve Graph {nodes [
   RigVMFunction_MathTransformArrayToSRT = &RigVMFunction_MathTransformArrayToSRT(Transforms=&RigUnit_GetTransformItemArray(Space=GlobalSpace, bInitial=False, Items=RigVMVariableNode() {variable=Bones}.Value).Transforms)
   RigVMFunction_Sequence = RigVMAggregateNode(ExecuteContext=@n2.ExecutePin)
   BSpline Passing thru 4 Points_1 = RigVMFunctionReferenceNode(
    Auto Weights=False
    Weight 1=0.333
    Weight 2=0.666
    Point 0=RigVMDispatchNode(Index=0, Array=@RigVMFunction_MathTransformArrayToSRT.Translations) {template="DISPATCH_RigVMDispatch_ArrayGetAtIndex(in Array,in Index,out Element)"}.Element
    Point 1=RigVMDispatchNode(Index=1, Array=@RigVMFunction_MathTransformArrayToSRT.Translations) {template="DISPATCH_RigVMDispatch_ArrayGetAtIndex(in Array,in Index,out Element)"}.Element
    Point 2=RigVMDispatchNode(Index=2, Array=@RigVMFunction_MathTransformArrayToSRT.Translations) {template="DISPATCH_RigVMDispatch_ArrayGetAtIndex(in Array,in Index,out Element)"}.Element
    Point 3=RigVMDispatchNode(Index=3, Array=@RigVMFunction_MathTransformArrayToSRT.Translations) {template="DISPATCH_RigVMDispatch_ArrayGetAtIndex(in Array,in Index,out Element)"}.Element
    Execute=@RigVMFunction_Sequence.C
   )
   n1 = RigVMVariableNode() {variable=IK Controls}
   Set Transform_2 = /Script/ControlRig.RigUnit_SetTranslation(
    Space=GlobalSpace
    Weight=1
    bInitial=False
    bPropagateToChildren=True
    Value=RigVMDispatchNode(Index=1, Array=@"BSpline Passing thru 4 Points_1".Control Points) {template="DISPATCH_RigVMDispatch_ArrayGetAtIndex(in Array,in Index,out Element)"}.Element
    Item=RigVMDispatchNode(Index=1, Array=@n1.Value) {template="DISPATCH_RigVMDispatch_ArrayGetAtIndex(in Array,in Index,out Element)"}.Element
    ExecutePin=@"BSpline Passing thru 4 Points_1".Execute
   ) {is=struct RigVMUnitNode, template="Set Transform::Execute(in Item,in Space,in bInitial,in Value,in Weight,in bPropagateToChildren)"}
   > Set Transform_2_1 = &RigUnit_SetTranslation(
    Space=GlobalSpace
    Weight=1
    bInitial=False
    bPropagateToChildren=True
    Value=RigVMDispatchNode(Index=2, Array=@"BSpline Passing thru 4 Points_1".Control Points) {template="DISPATCH_RigVMDispatch_ArrayGetAtIndex(in Array,in Index,out Element)"}.Element
    Item=RigVMDispatchNode(Index=2, Array=@n1.Value) {template="DISPATCH_RigVMDispatch_ArrayGetAtIndex(in Array,in Index,out Element)"}.Element
   ) {template="Set Transform::Execute(in Item,in Space,in bInitial,in Value,in Weight,in bPropagateToChildren)"}
   n2 = /Script/ControlRig.RigUnit_InverseExecution() {is=struct RigVMUnitNode}
   n3 = RigVMVariableNode() {variable=Bones}
   &RigUnit_SetTransform(
    Weight=1
    bInitial=False
    Value=&RigUnit_GetTransform(
     bInitial=false
     Item=RigVMDispatchNode(Index=-1, Array=@n3.Value) {template="DISPATCH_RigVMDispatch_ArrayGetAtIndex(in Array,in Index,out Element)"}.Element
    ).Transform
    ExecutePin=@n14.ExecutePin
    Item=RigVMVariableNode() {variable=Chest Space Null}.Value
   )
   n4 = RigVMVariableNode() {variable=IK Controls}
   n5 = RigVMVariableNode() {variable=Bones}
   &RigUnit_SetTransform(
    Value=(Rotation=(X=0,Y=0,Z=0,W=1),Translation=(X=0,Y=0,Z=0),Scale3D=(X=1,Y=1,Z=1))
    Weight=1
    bInitial=false
    Item=RigVMDispatchNode(Index=-1, Array=@n4.Value) {template="DISPATCH_RigVMDispatch_ArrayGetAtIndex(in Array,in Index,out Element)"}.Element
    Value.Translation=&RigUnit_GetTransform(
     bInitial=false
     Item=RigVMDispatchNode(Index=-1, Array=@n5.Value) {template="DISPATCH_RigVMDispatch_ArrayGetAtIndex(in Array,in Index,out Element)"}.Element
    ).Transform.Translation
    ExecutePin=@n13.ExecutePin
   )
   For_Each = RigVMDispatchNode(ExecuteContext=@n12.ExecuteContext, Array=RigVMVariableNode() {variable=FK Controls}.Value) {template="DISPATCH_RigVMDispatch_ArrayIterator(in Array,out Element,out Index,out Count,out Ratio)"}
   RigUnit_ItemArray = &RigUnit_ItemArray(Items=((Type=Bone,Name="spine_02"),(Type=Bone,Name="spine_03"),(Type=Bone,Name="spine_04")))
   At = RigVMDispatchNode(Index=@For_Each.Index, Array=@RigUnit_ItemArray.Items) {template="DISPATCH_RigVMDispatch_ArrayGetAtIndex(in Array,in Index,out Element)"}
   n6 = RigVMDispatchNode(Index=0, Array=@n4.Value) {template="DISPATCH_RigVMDispatch_ArrayGetAtIndex(in Array,in Index,out Element)"}
   n7 = RigVMDispatchNode(Index=0, Array=@n5.Value) {template="DISPATCH_RigVMDispatch_ArrayGetAtIndex(in Array,in Index,out Element)"}
   n8 = RigVMDispatchNode(Array=RigVMVariableNode() {variable=Bone Deltas}.Value, Index=@For_Each.Index) {template="DISPATCH_RigVMDispatch_ArrayGetAtIndex(in Array,in Index,out Element)"}
   Clone = RigVMDispatchNode(Array=RigVMVariableNode() {variable=Bones}.Value) {template="DISPATCH_RigVMDispatch_ArrayClone(in Array,out Clone)"}
   n9 = &RigUnit_GetTransform(bInitial=False, Item=@At.Element)
   n10 = &RigUnit_GetTransform(bInitial=false, Item=@n7.Element)
   Inverse = /Script/RigVM.RigVMFunction_MathTransformInverse(Value=@n8.Element) {is=struct RigVMUnitNode, template="Inverse::Execute(in Value,out Result)"}
   n11 = &RigVMFunction_MathTransformMul(A=@Inverse.Result, B=@n9.Transform) {template="Multiply::Execute(in A,in B,out Result)"}
   n12 = RigVMDispatchNode(Index=0, Array=@Clone.Clone, ExecuteContext=@RigVMFunction_Sequence.A) {template="DISPATCH_RigVMDispatch_ArrayRemove(io Array,in Index)"}
   n13 = &RigUnit_SetTransform(
    Value=(Rotation=(X=0,Y=0,Z=0,W=1),Translation=(X=0,Y=0,Z=0),Scale3D=(X=1,Y=1,Z=1))
    Weight=1
    bInitial=false
    Item=@n6.Element
    Value.Translation=@n10.Transform.Translation
    ExecutePin=@RigVMFunction_Sequence.B
   )
   n14 = &RigUnit_SetTransform(Weight=1, bInitial=False, Item=@For_Each.Element, ExecutePin=@For_Each.ExecuteContext, Value=@n11.Result)
  ]}
]
hierarchy [
 Root {type=Connector, loc=<0 0 0>, rot=<0 0 0 1>}
 Root {type=Socket, loc=<0 0 0>, rot=<0 0 0 1>}
 End {type=Connector, loc=<0 0 0>, rot=<0 0 0 1>}
 Parent {type=Connector, loc=<0 0 0>, rot=<0 0 0 1>}
]
variables [
 Bone Deltas {type="TArray<FTransform>"}
 Bones {type="TArray<FRigElementKey>"}
 Chest Control {type=FRigElementKey, default=(Type=Bone,Name="")}
 Chest Space Null {type=FRigElementKey, default=(Type=Bone,Name="")}
 Control Color {type=FLinearColor, default=(R=1,G=1,B=0,A=1)}
 Control Scale {type=double, default=1}
 FK Control Shape {type=FName, default=Box_Solid}
 FK Control Shape Transform {type=FTransform, default=(Rotation=(X=0,Y=0,Z=0,W=1),Translation=(X=0,Y=0,Z=0),Scale3D=(X=0.5,Y=0.5,Z=0.5))}
 FK Controls {type="TArray<FRigElementKey>"}
 Global Control Scale {type=double, default=1}
 Hip Control {type=FRigElementKey, default=(Type=Bone,Name="")}
 IK Channel {type=FRigElementKey, default=(Type=Bone,Name="")}
 IK Control Shape {type=FName, default=Default}
 IK Control Shape Transform {type=FTransform, default=(Rotation=(X=0,Y=0,Z=0,W=1),Translation=(X=0,Y=0,Z=0),Scale3D=(X=0.4,Y=0.4,Z=0.4))}
 IK Controls {type="TArray<FRigElementKey>"}
 IK Mode {type=bool, default=False}
 IK Offsets {type="TArray<FTransform>"}
 Length {type=double, default=0}
 Lengths {type="TArray<double>"}
 Negative Side {type=bool, default=False}
 Parent {type=FRigElementKey, default=(Type=None,Name="")}
 Pelvis Bone {type=FRigElementKey, default=(Type=None,Name="")}
 Running Parent {type=FRigElementKey, default=(Type=Bone,Name="")}
 Stretch {type=bool, default=False}
 Stretch Channel {type=FRigElementKey, default=(Type=Bone,Name="")}
]
